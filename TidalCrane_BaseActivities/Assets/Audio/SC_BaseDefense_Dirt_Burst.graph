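# Sound cue that plays a bug ground-burst sample, named as the dirt burst for the Tidal Crane base defense activity.
# A single wave player feeds a Modulator left at default settings, routed to the Fort explosion events sound class.
# Inline attenuation overrides give it a long falloff with distance-based low-pass filtering and occlusion muffling.
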SoundCue /TidalCrane_BaseActivities/Assets/Audio/SC_BaseDefense_Dirt_Burst
root=Modulator_0
nodes [
 {id=Modulator_0, children [WavePlayer_0]}
 {id=WavePlayer_0, properties {soundWaveAssetPtr=/TidalCrane_BaseActivities/Assets/Audio/waves/Bugs_Ground_Burst_01}}
]
properties {
 bOverrideAttenuation=true
 maxDistance=10000
 soundClassObject=/Game/Sounds/FortSoundClassesAndModes/Fort_Events_Explosions
 volumeMultiplier=1
 attenuationOverrides {
  bAttenuateWithLPF=true
  bEnableLogFrequencyScaling=true
  bEnableOcclusion=true
  falloffDistance=9000
  lPFFrequencyAtMin=100
  lPFRadiusMax=10000
  lPFRadiusMin=2000
  occlusionLowPassFilterFrequency=2000
  attenuationShapeExtents {x=1000}
 }
}
